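# Sound cue for the MatchMist turret's firing sound, layering close and distant variants of the shot.
# Each layer picks one of six waves at random, varies its pitch and volume, and runs through its own attenuation settings with overridden falloff.
# The distant layer adds a slight random delay and is mixed at half the close layer's level.
# The cue uses a medium explosives sound class and overrides concurrency to cap voices with distance-based volume scaling.
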SoundCue /DragonCartNarrative/Assets/Audio/Turret/SC_MatchMist_Turret_Fire
root=Mixer_8
nodes [
 {
  id=Attenuation_11
  children [Modulator_1]
  properties {attenuationSettings=&/ATT_MatchMist_Turret_Fire_Close, attenuationOverrides {falloffDistance=1600}}
 }
 {
  id=Attenuation_13
  children [Delay_1]
  properties {
   attenuationSettings=&/ATT_MatchMist_Turret_Fire_Distant
   attenuationOverrides {falloffDistance=3000, attenuationShapeExtents {x=1000}}
  }
 }
 {id=Delay_1, children [Modulator_0], properties {delayMax=0.05}}
 {id=Mixer_8, children [Attenuation_11, Attenuation_13], properties {inputVolume [0.5,0.25]}}
 {id=Modulator_0, children [Random_2], properties {pitchMax=1.25, pitchMin=0.75, volumeMin=0.75}}
 {id=Modulator_1, children [Random_0], properties {pitchMax=1.15, pitchMin=0.85, volumeMin=0.85}}
 {id=Random_0, children [WavePlayer_0,WavePlayer_1,WavePlayer_2,WavePlayer_3,WavePlayer_4,WavePlayer_5], properties {weights [1,1,1,1,1,1]}}
 {
  id=Random_2
  children [WavePlayer_12, WavePlayer_13, WavePlayer_14, WavePlayer_15, WavePlayer_16, WavePlayer_17]
  properties {weights [1,1,1,1,1,1]}
 }
 {id=WavePlayer_0, properties {soundWaveAssetPtr=/DragonCartNarrative/Assets/Audio/Turret/Soundwaves/MatchMist_Turret_Fire_Close_01}}
 {id=WavePlayer_1, properties {soundWaveAssetPtr=/DragonCartNarrative/Assets/Audio/Turret/Soundwaves/MatchMist_Turret_Fire_Close_02}}
 {id=WavePlayer_12, properties {soundWaveAssetPtr=/DragonCartNarrative/Assets/Audio/Turret/Soundwaves/MatchMist_Turret_Fire_Distant_01}}
 {id=WavePlayer_13, properties {soundWaveAssetPtr=/DragonCartNarrative/Assets/Audio/Turret/Soundwaves/MatchMist_Turret_Fire_Distant_02}}
 {id=WavePlayer_14, properties {soundWaveAssetPtr=/DragonCartNarrative/Assets/Audio/Turret/Soundwaves/MatchMist_Turret_Fire_Distant_03}}
 {id=WavePlayer_15, properties {soundWaveAssetPtr=/DragonCartNarrative/Assets/Audio/Turret/Soundwaves/MatchMist_Turret_Fire_Distant_04}}
 {id=WavePlayer_16, properties {soundWaveAssetPtr=/DragonCartNarrative/Assets/Audio/Turret/Soundwaves/MatchMist_Turret_Fire_Distant_05}}
 {id=WavePlayer_17, properties {soundWaveAssetPtr=/DragonCartNarrative/Assets/Audio/Turret/Soundwaves/MatchMist_Turret_Fire_Distant_06}}
 {id=WavePlayer_2, properties {soundWaveAssetPtr=/DragonCartNarrative/Assets/Audio/Turret/Soundwaves/MatchMist_Turret_Fire_Close_03}}
 {id=WavePlayer_3, properties {soundWaveAssetPtr=/DragonCartNarrative/Assets/Audio/Turret/Soundwaves/MatchMist_Turret_Fire_Close_04}}
 {id=WavePlayer_4, properties {soundWaveAssetPtr=/DragonCartNarrative/Assets/Audio/Turret/Soundwaves/MatchMist_Turret_Fire_Close_05}}
 {id=WavePlayer_5, properties {soundWaveAssetPtr=/DragonCartNarrative/Assets/Audio/Turret/Soundwaves/MatchMist_Turret_Fire_Close_06}}
]
properties {
 bOverrideConcurrency=true
 maxDistance=2097152
 soundClassObject=/Game/Sounds/FortSoundClassesAndModes/Fort_Explosives_Medium
 volumeMultiplier=0.85
 concurrencyOverrides {maxCount=8, volumeScale=0.75, volumeScaleMode=Distance}
}
