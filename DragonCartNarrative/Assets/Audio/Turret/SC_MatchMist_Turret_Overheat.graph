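# SoundCue for the MatchMist turret overheat sound, routed to an ambient props sound class.
# A Random node picks one of three equally weighted overheat waves, then a Modulator varies its pitch and slightly varies its volume.
# An Attenuation node applies the turret overheat attenuation asset with an overridden falloff distance.
# The cue boosts overall volume and overrides concurrency to cap simultaneous instances.
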SoundCue /DragonCartNarrative/Assets/Audio/Turret/SC_MatchMist_Turret_Overheat
root=Attenuation_11
nodes [
 {
  id=Attenuation_11
  children [Modulator_1]
  properties {attenuationSettings=&/ATT_MatchMist_Turret_Overheat, attenuationOverrides {falloffDistance=1600}}
 }
 {id=Modulator_1, children [Random_0], properties {pitchMax=1.15, pitchMin=0.85, volumeMin=0.85}}
 {id=Random_0, children [WavePlayer_6,WavePlayer_7,WavePlayer_8], properties {weights [1,1,1]}}
 {id=WavePlayer_6, properties {soundWaveAssetPtr=/DragonCartNarrative/Assets/Audio/Turret/Soundwaves/MatchMist_Turret_OverHeat_01}}
 {id=WavePlayer_7, properties {soundWaveAssetPtr=/DragonCartNarrative/Assets/Audio/Turret/Soundwaves/MatchMist_Turret_OverHeat_02}}
 {id=WavePlayer_8, properties {soundWaveAssetPtr=/DragonCartNarrative/Assets/Audio/Turret/Soundwaves/MatchMist_Turret_OverHeat_03}}
]
properties {
 maxDistance=32500
 soundClassObject=&Fort_Ambient_Props_Athena
 volumeMultiplier=2
 concurrencyOverrides {maxCount=5}
}
