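# Sound cue for the MatchMist turret's orb projectile impact, layering close and distant variants in the medium explosives sound class.
# A mixer blends the two layers at different input volumes, each with its own attenuation asset and falloff.
# Each layer picks randomly from six equally weighted waves, with modulators varying volume on both and adding pitch variation to the close layer.
# A concurrency override caps simultaneous instances and applies volume scaling.
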SoundCue /DragonCartNarrative/Assets/Audio/Turret/SC_MatchMist_Turret_Projectile_OrbImpact
root=Mixer_8
nodes [
 {
  id=Attenuation_11
  children [Modulator_1]
  properties {
   attenuationSettings=&/ATT_MatchMist_Turret_Projectile_OrbImpact_Close
   attenuationOverrides {falloffDistance=1600}
  }
 }
 {
  id=Attenuation_13
  children [Modulator_0]
  properties {
   attenuationSettings=&/ATT_MatchMist_Turret_Projectile_OrbImpact_Distant
   attenuationOverrides {falloffDistance=3000, attenuationShapeExtents {x=1000}}
  }
 }
 {id=Mixer_8, children [Attenuation_11, Attenuation_13], properties {inputVolume [0.7,0.4]}}
 {id=Modulator_0, children [Random_2], properties {volumeMin=0.5}}
 {id=Modulator_1, children [Random_0], properties {pitchMax=1.15, pitchMin=0.85, volumeMin=0.75}}
 {id=Random_0, children [WavePlayer_0,WavePlayer_1,WavePlayer_2,WavePlayer_3,WavePlayer_4,WavePlayer_5], properties {weights [1,1,1,1,1,1]}}
 {
  id=Random_2
  children [WavePlayer_6, WavePlayer_7, WavePlayer_8, WavePlayer_9, WavePlayer_10, WavePlayer_11]
  properties {weights [1,1,1,1,1,1]}
 }
 {id=WavePlayer_0, properties {soundWaveAssetPtr=/DragonCartNarrative/Assets/Audio/Turret/Soundwaves/MatchMist_OrbImpact_Close_01}}
 {id=WavePlayer_1, properties {soundWaveAssetPtr=/DragonCartNarrative/Assets/Audio/Turret/Soundwaves/MatchMist_OrbImpact_Close_02}}
 {id=WavePlayer_10, properties {soundWaveAssetPtr=/DragonCartNarrative/Assets/Audio/Turret/Soundwaves/MatchMist_OrbImpact_Distant_05}}
 {id=WavePlayer_11, properties {soundWaveAssetPtr=/DragonCartNarrative/Assets/Audio/Turret/Soundwaves/MatchMist_OrbImpact_Distant_06}}
 {id=WavePlayer_2, properties {soundWaveAssetPtr=/DragonCartNarrative/Assets/Audio/Turret/Soundwaves/MatchMist_OrbImpact_Close_03}}
 {id=WavePlayer_3, properties {soundWaveAssetPtr=/DragonCartNarrative/Assets/Audio/Turret/Soundwaves/MatchMist_OrbImpact_Close_04}}
 {id=WavePlayer_4, properties {soundWaveAssetPtr=/DragonCartNarrative/Assets/Audio/Turret/Soundwaves/MatchMist_OrbImpact_Close_05}}
 {id=WavePlayer_5, properties {soundWaveAssetPtr=/DragonCartNarrative/Assets/Audio/Turret/Soundwaves/MatchMist_OrbImpact_Close_06}}
 {id=WavePlayer_6, properties {soundWaveAssetPtr=/DragonCartNarrative/Assets/Audio/Turret/Soundwaves/MatchMist_OrbImpact_Distant_01}}
 {id=WavePlayer_7, properties {soundWaveAssetPtr=/DragonCartNarrative/Assets/Audio/Turret/Soundwaves/MatchMist_OrbImpact_Distant_02}}
 {id=WavePlayer_8, properties {soundWaveAssetPtr=/DragonCartNarrative/Assets/Audio/Turret/Soundwaves/MatchMist_OrbImpact_Distant_03}}
 {id=WavePlayer_9, properties {soundWaveAssetPtr=/DragonCartNarrative/Assets/Audio/Turret/Soundwaves/MatchMist_OrbImpact_Distant_04}}
]
properties {
 bOverrideConcurrency=true
 maxDistance=2097152
 soundClassObject=/Game/Sounds/FortSoundClassesAndModes/Fort_Explosives_Medium
 volumeMultiplier=1
 concurrencyOverrides {maxCount=3, volumeScale=0.5241, volumeScaleAttackTime=0.1386}
}
